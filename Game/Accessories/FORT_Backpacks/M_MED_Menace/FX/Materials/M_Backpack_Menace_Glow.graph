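# Unlit additive Niagara sprite material for the glow on the Menace backpack.
# Two soft flare samples spin in opposite directions at different rates around slightly off-center pivots, with per-particle random phase.
# Their average is squared to modulate a center-scaled spikeball flare, giving a shifting shimmer.
# Depth fade softens the result, which is multiplied by particle color and alpha.
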Material /Game/Accessories/FORT_Backpacks/M_MED_Menace/FX/Materials/M_Backpack_Menace_Glow
graph {
 properties {bUsedWithNiagaraSprites=true, blendMode=BLEND_Additive, pixelDepthOffsetMode=PDOM_Legacy, shadingModel=MSM_Unlit, usedShadingModels=Unlit}
 nodes [
  n1 = ParticleColor()
  n2 = MaterialFunctionCall(UVs=TextureCoordinate(), Texture Scale=Constant() {r=4}) {materialFunction=&ScaleUVsByCenter}
  n3 = TextureSample(UVs=@n2.UVs) {texture=/Game/Effects/Fort_Effects/Textures/Flares/T_Flare_Spikeball}
  n4 = Add(A=Multiply(A=Time(), B=ScalarParameter() {defaultValue=0.333, parameterName=Rotator Speed}), B=ParticleRandom())
  n5 = Rotator(Time=@n4) {centerX=0.495, centerY=0.505, speed=6.28}
  n6 = Rotator(Coordinate=OneMinus(None=@n2.UVs), Time=Multiply(A=@n4) {constB=-0.618}) {centerX=0.505, centerY=0.495, speed=6.28}
  n7 = Divide(A=Add(
    A=TextureSample(UVs=@n5) {texture=/Game/Effects/Fort_Effects/Textures/Flares/T_Flare_Soft}.B
    B=TextureSample(UVs=@n6) {texture=&T_Flare_Soft}.B
   ))
  n8 = DepthFade(Opacity=Multiply(A=@n3.B, B=Multiply(A=@n7, B=@n7)), FadeDistance=ScalarParameter() {defaultValue=11, parameterName=Depth Fade Distance})
  LinearInterpolate(Alpha=Multiply(A=@n1.A, B=@n8)) {constA=1}
  MaterialFunctionCall(Emissive=@n1.RGB) {materialFunction=&MF_FX_EmissiveAdjustments}
  Multiply(A=@n1.RGB, B=@n8)
  Multiply(A=@n3.B, B=@n3.B)
 ]
}
